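# Unlit translucent master material for index-of-refraction distortion on Niagara meshes, sprites and particles.
# Two tileable polar-coordinate UV sets, offset by Dynamic Parameters, sample a gradient refraction mask and a swirl-noise normal map.
# Particle alpha times the mask blends a refraction value from 1 toward 0.5, with additional round-square mask math.
# The normal sample takes mip derivatives from the original UVs rather than the polar ones.
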Material /DealershipSharedLibrary/FX/Materials/MasterMaterials/Refraction/M_DLR_Refraction
graph {
 properties {
  bUsedWithNiagaraMeshParticles=true
  bUsedWithNiagaraSprites=true
  bUsedWithParticleSprites=true
  bUsedWithSkeletalMesh=true
  blendMode=BLEND_Translucent
  pixelDepthOffsetMode=PDOM_Legacy
  refractionMethod=RM_IndexOfRefraction
  shadingModel=MSM_Unlit
  usedShadingModels=Unlit
 }
 nodes [
  n1 = TextureCoordinate()
  n2 = MaterialFunctionCall(XY=MaterialFunctionCall!(UV Coord=@n1, LenghtScale=ScalarParameter() {defaultValue=1, parameterName=Radisl Tile}, RadialScale) {materialFunction=/Game/Global/MaterialFunctions/TechArt/MF_PolarCoordinates}.Result) {materialFunction=&Swizzle}
  Add(A=@n2.YX)
  Constant() {r=1}
  n3 = DynamicParameter() {defaultValue=<rgba 15.8663 0.1691 2 2>, paramNames [Param1, RefractionOffset, Param3, Param4]}
  n4 = OneMinus(None=@n3.RefractionOffset)
  n5 = TextureSampleParameter2D(UVs=Add(A=@n2.YX, B=AppendVector(A=@n3.Param1, B=@n4))) {parameterName=RefractionMask, samplerType=SAMPLERTYPE_Masks, texture=/Game/Global/Textures/Default/Masks_and_Gradients/T_VertGradient_MASK}
  n6 = LinearInterpolate(B=Constant() {r=0.5}, Alpha=Multiply(A=ParticleColor().A, B=@n5.R)) {constA=1}
  n7 = TextureSample() {texture=/Game/Effects/Fort_Effects/Gameplay/Tutorial_Target/shader_mask_round_square_blurry}
  n8 = Constant() {r=4}
  n9 = Saturate(None=Power(Base=Multiply(A=@n7.G, B=@n8)) {constExponent=1.5})
  LinearInterpolate(B=@n6, Alpha=@n9) {constA=1}
  Multiply(A=@n9, B=@n6)
  Multiply(A=Subtract(A=@n7.G, B=@n8), B=@n5.R)
  OneMinus(None=@n7.G)
  n10 = MaterialFunctionCall(
   LenghtScale=ScalarParameter() {defaultValue=1, parameterName=Nrml Radisl Tile}
   RadialScale=ScalarParameter() {defaultValue=2, parameterName=Nrml Radisl TileV}
  ) {materialFunction=&MF_PolarCoordinates}
  n11 = OneMinus(None=@n3.Param4)
  TextureSample!(
   UVs=Add(
    A=MaterialFunctionCall(XY=@n10.Result) {materialFunction=&Swizzle}.YX
    B=AppendVector(A=@n3.Param3, B=@n11)
   )
   Tex
   "DDX(UVs)"=DDX(Value=@n1)
   "DDY(UVs)"=DDY(Value=@n1)
   Apply View MipBias
  ) {mipValueMode=TMVM_Derivative, samplerType=SAMPLERTYPE_Normal, texture=/Game/Effects/Fort_Effects/Emotes/BirdNestBlue/T_SwirlNoise}
 ]
}
